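# Editor utility widget that serves as a front end for the Fortnite cosmetic placement tool, based on FortCosmeticPlacementTool.
# It shows a single Calculate button in a vertical box.
# Clicking it calls GetPossibleBackblingCombinationsCount on the native parent and prints the resulting integer as a string.
# Construct, PreConstruct and Tick events are left empty.
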EditorUtilityWidgetBlueprint /Game/ArtTools/Tech_Art/CosmeticPlacementTool/EUW_CosmeticPlacementTool
graphs [EventGraph {nodes [
   "On Clicked (CalculateButton)"() {is=title K2Node_ComponentBoundEvent}
   > n1 = GetPossibleBackblingCombinationsCount!(execute, self:Fort Cosmetic Placement Tool Object Reference) -> (then, ReturnValue:int=0)
   > PrintString(InString="To String (Integer)"(InInt=@n1))
   Event PreConstruct()
   Event Construct()
   Event Tick()
  ]}]
variables []
widgets {
 info {parentClass=/Script/FortniteEditor.FortCosmeticPlacementTool, rootWidgetClass=&VerticalBox, widgetCount=3}
 widgets [
  VerticalBox_0 {widgetClassPath=&VerticalBox, widget=VerticalBox_0}
  CalculateButton {
   widgetClassPath=&EditorUtilityButton
   parent=VerticalBox_0
   widget=CalculateButton
   slot {class=VerticalBoxSlot, properties {verticalAlignment=VAlign_Center, size {sizeRule=Fill, value=1}}}
  }
  TextBlock_35 {
   widgetClassPath=&TextBlock
   parent=CalculateButton
   widget=TextBlock_35
   properties {justification=Center, text=Calculate}
   slot {class=ButtonSlot, properties {horizontalAlignment=HAlign_Fill}}
  }
 ]
}
